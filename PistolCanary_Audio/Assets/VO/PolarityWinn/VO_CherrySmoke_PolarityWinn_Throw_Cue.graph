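# Sound cue for PolarityWinn's voice lines when throwing a Cherry Smoke, apparently as an NPC boss combat bark.
# A Random node picks among seven equally weighted wave players, each holding one throw take.
# Playback uses the shared alerted-NPC VO attenuation, the enemy-critical VO sound class and boss combat chatter concurrency to limit overlapping barks.
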SoundCue /PistolCanary_Audio/Assets/VO/PolarityWinn/VO_CherrySmoke_PolarityWinn_Throw_Cue
root=Random_0
nodes [
 {
  id=Random_0
  children [WavePlayer_0,WavePlayer_1,WavePlayer_2,WavePlayer_3,WavePlayer_4,WavePlayer_5,WavePlayer_6]
  properties {weights [1,1,1,1,1,1,1]}
 }
 {id=WavePlayer_0, properties {soundWaveAssetPtr=/PistolCanary_Audio/Assets/VO/PolarityWinn/wavs/VO_CherrySmoke_PolarityWinn_Throw_010}}
 {id=WavePlayer_1, properties {soundWaveAssetPtr=/PistolCanary_Audio/Assets/VO/PolarityWinn/wavs/VO_CherrySmoke_PolarityWinn_Throw_011}}
 {id=WavePlayer_2, properties {soundWaveAssetPtr=/PistolCanary_Audio/Assets/VO/PolarityWinn/wavs/VO_CherrySmoke_PolarityWinn_Throw_012}}
 {id=WavePlayer_3, properties {soundWaveAssetPtr=/PistolCanary_Audio/Assets/VO/PolarityWinn/wavs/VO_CherrySmoke_PolarityWinn_Throw_013}}
 {id=WavePlayer_4, properties {soundWaveAssetPtr=/PistolCanary_Audio/Assets/VO/PolarityWinn/wavs/VO_CherrySmoke_PolarityWinn_Throw_014}}
 {id=WavePlayer_5, properties {soundWaveAssetPtr=/PistolCanary_Audio/Assets/VO/PolarityWinn/wavs/VO_CherrySmoke_PolarityWinn_Throw_015}}
 {id=WavePlayer_6, properties {soundWaveAssetPtr=/PistolCanary_Audio/Assets/VO/PolarityWinn/wavs/VO_CherrySmoke_PolarityWinn_Throw_016}}
]
properties {
 attenuationSettings=/Game/Sounds/Fort_Human_VO_BR/NPCs/Shared/NPC_VO_Alerted_Att
 maxDistance=12500
 soundClassObject=&Fort_VO_Enemy_Critical
 volumeMultiplier=1
 concurrencySet [&NPC_Boss_Combat_Chatter_Conc]
}
